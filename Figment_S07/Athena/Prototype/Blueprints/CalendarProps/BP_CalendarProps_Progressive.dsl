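; Calendar-driven prop spawner that places saved actors according to how far a live calendar event has progressed.
; After match start, the authority watches the named event and sets progress from elapsed time, or to full if forced.
; Its OnRep spawns props whose appear/disappear window contains that progress, server or client side, batched across frames.
; Edit and Save/Harvest events convert between placed actors and stored relative-transform data, with construction-script visibility preview.
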(event EventBeginPlay
  (Networking|SwitchHasAuthority
    (:Authority
      (bind _asfort_game_mode_zone (Utilities|Casting|CastToFortGameModeZone (Game|GetGameMode)))
      (EventDispatchers|BindEventtoOnHandleMatchHasStarted _asfort_game_mode_zone (delegate Custom|BindToCalendarEvents)))))

(event Custom|BindToCalendarEvents
  ;[1 1 1] is this active? If yes how far along are we?
  (Networking|SwitchHasAuthority
    (:Authority
      (bind (_eventname _timeuntilend _timesincebegin) (MCPContext|WatchCalendarEvent (Variables|Default|GetEventName))
        (:OnEventActive
          (if (Variables|Default|GetAllowSpawn)
            (if (Variables|Default|GetAlways%20Spawn%20When%20Active)
              (Variables|Default|SetCalendarEventProgress 1.0)
              (else
                (Variables|Default|SetCalendarEventProgress (Math|Timespan|TimespanRatio _timesincebegin (Math|Timespan|Timespan+Timespan _timeuntilend _timesincebegin)))))))))))

(event Custom|SpawnActors
  (bind _actordata (Variables|Hidden|GetActorData))
  ;[1 1 1] Spawn Props
  (if (> (Variables|Default|GetCalendarEventProgress) 0.0)
    (for _index (range (Variables|Default|GetSpawnProgress) (+ (Math|Integer|Min(Integer) (Utilities|Array|LastIndex (Variables|Hidden|GetActorData)) (+ (Variables|Default|GetSpawnProgress) (Variables|Default|GetSpawnPerFrame))) 1))
      (bind (_output_actor_27_6a42496a4663e346c97afeb9993c68f7 _output_transform_11_ce6eb422421fe49e3943bdb094b64c74 _output_progressappear_21_5b07da4e4ffa7913e1c2ef864c0bdc68 _output_progressdisappear_22_d8fc05f84ecb57e73362de9563548198) (Utilities|Struct|BreakCalendarPropData (Utilities|Array|Get(acopy) _actordata _index)))
      (if (and (<= _output_progressappear_21_5b07da4e4ffa7913e1c2ef864c0bdc68 (Variables|Default|GetCalendarEventProgress)) (>= _output_progressdisappear_22_d8fc05f84ecb57e73362de9563548198 (Variables|Default|GetCalendarEventProgress)))
        (Game|SpawnActorfromClass (Utilities|ResolveSoftReference _output_actor_27_6a42496a4663e346c97afeb9993c68f7) (Math|Transform|ComposeTransforms _output_transform_11_ce6eb422421fe49e3943bdb094b64c74 (Transformation|GetActorTransform)) "AlwaysSpawn" "SelectDefaultAtRuntime" 0 0)))
    (if (not (>= (stale _index) (Utilities|Array|LastIndex (Variables|Hidden|GetActorData))))
      (Variables|Default|SetSpawnProgress (+ (Variables|Default|GetSpawnProgress) (Variables|Default|GetSpawnPerFrame)))
      (Utilities|FlowControl|Delay 0.0)
      (CallFunction|SpawnActors))))

(event Custom|Save/HarvestActors
  ;[0.0916 0.2553 1] Save
  (if (Variables|Default|GetPreviewMode)
    _
    (elif (Variables|Hidden|GetEditMode(Don'tTouch))
      (if (Variables|Hidden|GetClear%20Actor%20Data%20with%20Harvest%20%28advanced%29)
        (Utilities|Array|Clear (Variables|Hidden|GetActorData)))
      (for _array_element (Variables|Default|GetActors)
        (bind (_array_element_actor_17_6a42496a4663e346c97afeb9993c68f7 _array_element_progressappear_19_5b07da4e4ffa7913e1c2ef864c0bdc68 _array_element_progressdisappear_21_c4e989dc4ac7b7e7064d06a0a4c655ec) (Utilities|Struct|BreakCalendarPropSelectionData _array_element))
        (Utilities|Array|Add (Variables|Hidden|GetActorData) (Utilities|Struct|MakeCalendarPropData (Utilities|ResolveSoftReference (Utilities|GetClass _array_element_actor_17_6a42496a4663e346c97afeb9993c68f7)) (Math|Transform|MakeRelativeTransform (Transformation|GetActorTransform _array_element_actor_17_6a42496a4663e346c97afeb9993c68f7) (Transformation|GetActorTransform)) _array_element_progressappear_19_5b07da4e4ffa7913e1c2ef864c0bdc68 _array_element_progressdisappear_21_c4e989dc4ac7b7e7064d06a0a4c655ec))
        (Actor|DestroyActor _array_element_actor_17_6a42496a4663e346c97afeb9993c68f7))
      (Variables|Hidden|SetEditMode(Don'tTouch) false)
      (Utilities|Array|Clear (Variables|Default|GetActors)))))

(event Custom|EditActors
  ;[0.0916 0.2553 1] Edit
  (if (Variables|Default|GetPreviewMode)
    _
    (elif (Variables|Hidden|GetEditMode(Don'tTouch))
      _
      (else
        (for _array_element (Variables|Hidden|GetActorData)
          (bind (_array_element_actor_27_6a42496a4663e346c97afeb9993c68f7 _array_element_transform_11_ce6eb422421fe49e3943bdb094b64c74 _array_element_progressappear_21_5b07da4e4ffa7913e1c2ef864c0bdc68) (Utilities|Struct|BreakCalendarPropData _array_element))
          (bind _returnvalue (Game|SpawnActorfromClass (Utilities|ResolveSoftReference _array_element_actor_27_6a42496a4663e346c97afeb9993c68f7) (Math|Transform|ComposeTransforms _array_element_transform_11_ce6eb422421fe49e3943bdb094b64c74 (Transformation|GetActorTransform)) "AlwaysSpawn" "SelectDefaultAtRuntime" 0 0))
          (Utilities|Array|Add (Variables|Default|GetActors) (Utilities|Struct|MakeCalendarPropSelectionData _returnvalue _array_element_progressappear_21_5b07da4e4ffa7913e1c2ef864c0bdc68)))
        (Variables|Hidden|SetEditMode(Don'tTouch) true)))))

(fn OnRep_CalendarEventProgress
  (if (Variables|Default|GetServerSideSpawning)
    (Networking|SwitchHasAuthority
      (:Authority
        (CallFunction|SpawnActors)))
    (else
      (Networking|SwitchHasAuthority
        (:Remote
          (CallFunction|SpawnActors))))))

(fn OnRep_GrowthStage
  (CallFunction|SpawnActors))

(fn ConstructionScript
  (if (Variables|Hidden|GetEditMode(Don'tTouch))
    (for _array_element (Variables|Default|GetActors)
      (bind (_array_element_actor_17_6a42496a4663e346c97afeb9993c68f7 _array_element_progressappear_19_5b07da4e4ffa7913e1c2ef864c0bdc68 _array_element_progressdisappear_21_c4e989dc4ac7b7e7064d06a0a4c655ec) (Utilities|Struct|BreakCalendarPropSelectionData _array_element))
      (if (< (Variables|Default|GetProgressPreview) 1.0)
        (if (and (<= _array_element_progressappear_19_5b07da4e4ffa7913e1c2ef864c0bdc68 (Variables|Default|GetProgressPreview)) (>= _array_element_progressdisappear_21_c4e989dc4ac7b7e7064d06a0a4c655ec (Variables|Default|GetProgressPreview)))
          (Rendering|SetVisibility (Variables|/Script/Engine.Actor|GetRootComponent _array_element_actor_17_6a42496a4663e346c97afeb9993c68f7) true true)
          (else
            (Rendering|SetVisibility (Variables|/Script/Engine.Actor|GetRootComponent _array_element_actor_17_6a42496a4663e346c97afeb9993c68f7) false true)))
        (else
          (Rendering|SetVisibility (Variables|/Script/Engine.Actor|GetRootComponent _array_element_actor_17_6a42496a4663e346c97afeb9993c68f7) true true))))
    (else
      (Variables|Default|SetPreviewMode false))))
