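; BP_DataLayerSwitch is a prototype Blueprint that picks between an original and a calendar World Partition data layer based on a calendar event.
; On BeginPlay it watches the event; if active, it activates the calendar layer and unloads the original.
; The next sequence step applies the reverse as default.
; A shared ValueSet? flag lets only the first switch apply, so activation reported after that default changes nothing.
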(event EventBeginPlay
  (Utilities|FlowControl|Sequence
    (:then_0
      (MCPContext|WatchCalendarEvent (Variables|Default|GetEvent%20to%20Watch)
        (:OnEventActive
          (if (not (Variables|Default|GetValueSet?))
            (bind _returnvalue (WorldPartition|GetDataLayerManager))
            (DataLayers|SetDataLayerInstanceRuntimeState _returnvalue (DataLayers|GetDataLayerInstancefromAsset _returnvalue (Variables|Default|GetOriginal%20Data%20Layer)) "Unloaded" false)
            (DataLayers|SetDataLayerInstanceRuntimeState _returnvalue (DataLayers|GetDataLayerInstancefromAsset _returnvalue (Variables|Default|GetCalendar%20Data%20Layer)) "Activated" false)
            (Variables|Default|SetValueSet? true)))))
    (:then_1
      (if (not (Variables|Default|GetValueSet?))
        (bind _returnvalue_2 (WorldPartition|GetDataLayerManager))
        (DataLayers|SetDataLayerInstanceRuntimeState _returnvalue_2 (DataLayers|GetDataLayerInstancefromAsset _returnvalue_2 (Variables|Default|GetCalendar%20Data%20Layer)) "Unloaded" false)
        (DataLayers|SetDataLayerInstanceRuntimeState _returnvalue_2 (DataLayers|GetDataLayerInstancefromAsset _returnvalue_2 (Variables|Default|GetOriginal%20Data%20Layer)) "Activated" false)
        (Variables|Default|SetValueSet? true)))))

(fn ConstructionScript)
