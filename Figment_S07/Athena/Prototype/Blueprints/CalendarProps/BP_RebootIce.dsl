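; BP_RebootIce is a calendar-driven prototype prop that destroys its referenced RebootDisappear actor when a calendar event fires.
; On BeginPlay, the server binds to the game mode's match-start dispatcher, which then starts watching calendar event "C0", labeled Beacon Event.
; When that event becomes active, it destroys the actor and sets ValueSet, the flag checked before watching.
; A separate VanishReboot event duplicates this removal logic.
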(event EventBeginPlay
  (Networking|SwitchHasAuthority
    (:Authority
      (bind _asfort_game_mode_zone (Utilities|Casting|CastToFortGameModeZone (Game|GetGameMode)))
      (EventDispatchers|BindEventtoOnHandleMatchHasStarted _asfort_game_mode_zone (delegate Custom|BindToCalendarEvents)))))

(event Custom|VanishReboot
  (Actor|DestroyActor (Variables|Default|GetRebootDisappear))
  (Variables|Default|SetValueSet? true))

(event Custom|BindToCalendarEvents
  (if (not (Variables|Default|GetValueSet?))
    (MCPContext|WatchCalendarEvent "C0"
      (:OnEventActive
        (Actor|DestroyActor (Variables|Default|GetRebootDisappear))
        (Variables|Default|SetValueSet? true))) ; Beacon Event
  ))

(fn ConstructionScript)
