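; BP_TiltedSwap is a calendar prop that swaps Tilted between its new and destroyed World Partition data layers.
; On BeginPlay it watches the S7_TiltedDestroyed calendar event to activate the destroyed layer, then defaults to the new layer if nothing is set.
; A shared ValueSet flag latches the first choice, so an event activating after the default is ignored.
; Custom events TiltedNew and TiltedDestroyed force either state without checking the flag.
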(event Custom|TiltedDestroyed
  (bind _returnvalue (WorldPartition|GetDataLayerManager))
  (DataLayers|SetDataLayerInstanceRuntimeState _returnvalue (DataLayers|GetDataLayerInstancefromAsset _returnvalue (Variables|Default|GetTilted_New)) "Unloaded" false)
  (DataLayers|SetDataLayerInstanceRuntimeState _returnvalue (DataLayers|GetDataLayerInstancefromAsset _returnvalue (Variables|Default|GetTilted_Destroyed)) "Activated" false)
  (Variables|Default|SetValueSet? true))

(event EventBeginPlay
  (Utilities|FlowControl|Sequence
    (:then_0
      ;[1 1 1] Switch out for Getting Crack 35 Progression
      (MCPContext|WatchCalendarEvent "S7_TiltedDestroyed"
        (:OnEventActive
          (if (not (Variables|Default|GetValueSet?))
            (bind _returnvalue (WorldPartition|GetDataLayerManager))
            (DataLayers|SetDataLayerInstanceRuntimeState _returnvalue (DataLayers|GetDataLayerInstancefromAsset _returnvalue (Variables|Default|GetTilted_New)) "Unloaded" false)
            (DataLayers|SetDataLayerInstanceRuntimeState _returnvalue (DataLayers|GetDataLayerInstancefromAsset _returnvalue (Variables|Default|GetTilted_Destroyed)) "Activated" false)
            (Variables|Default|SetValueSet? true)))))
    (:then_1
      (if (not (Variables|Default|GetValueSet?))
        (bind _returnvalue_2 (WorldPartition|GetDataLayerManager))
        (DataLayers|SetDataLayerInstanceRuntimeState _returnvalue_2 (DataLayers|GetDataLayerInstancefromAsset _returnvalue_2 (Variables|Default|GetTilted_Destroyed)) "Unloaded" false)
        (DataLayers|SetDataLayerInstanceRuntimeState _returnvalue_2 (DataLayers|GetDataLayerInstancefromAsset _returnvalue_2 (Variables|Default|GetTilted_New)) "Activated" false)
        (Variables|Default|SetValueSet? true)))))

(event Custom|TiltedNew
  (bind _returnvalue (WorldPartition|GetDataLayerManager))
  (DataLayers|SetDataLayerInstanceRuntimeState _returnvalue (DataLayers|GetDataLayerInstancefromAsset _returnvalue (Variables|Default|GetTilted_Destroyed)) "Unloaded" false)
  (DataLayers|SetDataLayerInstanceRuntimeState _returnvalue (DataLayers|GetDataLayerInstancefromAsset _returnvalue (Variables|Default|GetTilted_New)) "Activated" false)
  (Variables|Default|SetValueSet? true))

(fn ConstructionScript)
